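# Skydiving contrail cosmetic in which spray cans circle the player and spray coloured paint trails.
# The Spraycan emitter spawns mesh cans on a ring in three colour groups, with animated tilt.
# Liquid sprites, ribbons, smoke and drops read the cans through attribute readers, spawning at nozzle offsets and driving velocity from can orientation and age.
# ParaGlideLeanAlpha adjusts rates, lifetimes, radius and speed, while front-end modifiers retune motion outside gameplay.
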NiagaraSystem /BRCosmetics/Athena/Effects/Particles/Skydiving/NS_Contrail_GraffitiTon/NS_Contrail_GraffitiTon_Trail_v3
dependencies {
 usedDataInterfaces [
  &NiagaraDataInterfaceArrayColor
  &NiagaraDataInterfaceCurve
  &NiagaraDataInterfaceParticleRead
  &NiagaraDataInterfaceSpriteRendererInfo
 ]
 usedDynamicInputs [
  &self:SpawnGrp_0_1_Mask_06
  &MakeCustomLinearColorFromBool
  &TransformVector
  &SetBoolByFloatComparison
 ]
 usedModules [
  &SystemState
  &self:SystemSpawnScript.SystemScriptSource.SystemScriptGraph.NiagaraNodeAssignment_2.SetVariables_#1
  &self:SystemSpawnScript.SystemScriptSource.SystemScriptGraph.NiagaraNodeAssignment_3.SetVariables_#2
  &self:SystemSpawnScript.SystemScriptSource.SystemScriptGraph.NiagaraNodeAssignment_1.SetVariables_#3
  &InterpolateOverTime
  &FN_Integrator
  &Timeline
  &self:SystemSpawnScript.SystemScriptSource.SystemScriptGraph.NiagaraNodeAssignment_4.SetVariables_#4
  /Game/Effects/Niagara/Modules/Skydiving/FN_DivingAttributes/v5/FN_DivingAttributes
 ]
}
emitters {
 Liquid_Particles001 {
  emitterUpdateScript [
   &EmitterState()
   &SpawnParticlesFromOtherEmitter(
    Attribute Reader=DataInterface() {propertyValues {EmitterBinding {bindingMode=Other, emitterName=Spraycan}}}
    Calculate Spawn Rate Per Particle=false
    Spawn Rate=50
    Use Maximum Age=false
    Use Minimum Age=false
   )
  ]
  particleSpawnScript [
   &InitializeParticle(
    Color Mode=::NewEnumerator0
    Lifetime=&Lerp_Float(A=0.2, B=0.5, Alpha=$User.ParaGlideLeanAlpha)
    Position=<0 0 32.0451>
    Position Mode=::NewEnumerator1
    Position Offset=<0 0 31.81>
    Position Offset Coordinate Space=::Local
    Sprite Size Mode=::NewEnumerator1
    Uniform Sprite Size Max=5
    Uniform Sprite Size Min=1
    UsePositionOffset=true
    Write Material Random=true
   )
   &SampleParticlesFromOtherEmitter(
    Age Sampling:ENiagara_DirectReadSamplingMode=::NewEnumerator1
    Applied Velocity Scale=0
    Color Sampling=::NewEnumerator0
    Lifetime Sampling:ENiagara_DirectReadSamplingMode=::NewEnumerator1
    Maximum Sampled Particle Age=1.45
    Mesh Orientation Sampling:ENiagara_DirectReadSamplingMode=::NewEnumerator0
    Minimum Sampled Particle Age=0.1
    Module.MaxSampledAgeCutoff=true
    Module.MinSampledAgeCutoff=false
    Particle ID Sampling=::NewEnumerator1
    Reject Particle By Sampled Age=true
    Ribbon UV Distance:ENiagara_DirectReadSamplingMode=::NewEnumerator0
    Velocity Sampling=::NewEnumerator0
   )
   /Niagara/Modules/Update/SubUV/SubUVAnimation(Number Of Frames=8, SubUV Animation Mode=::NewEnumerator1) {module=SubUVAnimation001}
   &CameraOffset()
   /SetVariables(Particles.Position=&AddVectorToPosition(
     Position=$Particles.Position
     Vector=&Multiply_VectorByFloat(
      Vector=&ReturnMeshOrientationAxis(Mesh Orientation Axis=::ZAxis, Mesh Orientation Quat=$Particles.MeshOrientation)
      Float=13.9008
     )
    )) {isSetParameters=true}
   /SetVariables(Particles.SprayDirection:Vector3f=&ReturnMeshOrientationAxis(Mesh Orientation Axis=::YAxis, Mesh Orientation Quat=$Particles.MeshOrientation)) {isSetParameters=true}
   /SetVariables(Particles.SprayDirection:Vector3f=&NormalizeVector(
     Vector To Normalize=&MakeVector(
      X=&MakeFloatFromVector(Vector=$Particles.SprayDirection, Channel=::NewEnumerator0)
      Y=&MakeFloatFromVector(Vector=$Particles.SprayDirection, Channel=::NewEnumerator1)
      Z=0
     )
     Safe Normalize=true
     Safe Fallback Threshold=0
     Safe Fallback Vector=<0 0 0.001>
    )) {isSetParameters=true}
  ]
  particleUpdateScript [
   &ParticleState()
   &InheritSourceMovement(Applied Position Scale=<-0.1 -0.1 0>, ApplyForce=false, ApplyPosition=true, ApplyPositionFalloff=false)
   /Niagara/Modules/AttributeReader/UpdateParticlesFromOtherEmitter(
    Age Sampling:ENiagara_DirectReadSamplingMode=::NewEnumerator1
    Attribute Reader=DataInterface() {propertyValues {EmitterBinding {bindingMode=Other, emitterName=Spraycan}}}
    Lifetime Sampling:ENiagara_DirectReadSamplingMode=::NewEnumerator1
    Mesh Orientation Sampling:ENiagara_DirectReadSamplingMode=::NewEnumerator0
    Position Sampling:ENiagara_DirectReadSamplingMode=::NewEnumerator1
    Velocity Sampling:ENiagara_DirectReadSamplingMode=::NewEnumerator1
   )
   &ScaleSpriteSize(Scale Factor=&Vector2DFromFloat(Value=&FloatFromCurve(
      FloatCurve=DataInterface() {propertyValues {Curve {keys [
          {arriveTangent=2.9574, interpMode=RCIM_Cubic, leaveTangent=2.9574, tangentMode=RCTM_User, time=0, value=0.2}
          {arriveTangent=6.5547, interpMode=RCIM_Cubic, leaveTangent=6.5547, tangentMode=RCTM_User, time=0.7, value=3}
          {arriveTangent=-24.1897, interpMode=RCIM_Cubic, leaveTangent=-24.1897, tangentMode=RCTM_User, time=1, value=0}
         ]}}}
      Scale Curve=0.7
     ))) {module=ScaleSpriteSize002}
   /SetVariables(Particles.ToNozzle:Vector3f=&ReturnMeshOrientationAxis(Mesh Orientation Axis=::ZAxis, Mesh Orientation Quat=$Particles.MeshOrientation)) {isSetParameters=true}
   /SetVariables(Particles.ToNozzleHorizontal:Vector3f=&NormalizeVector(
     Vector To Normalize=&MakeVector(
      X=&MakeFloatFromVector(Vector=$Particles.ToNozzle, Channel=::NewEnumerator0)
      Y=&MakeFloatFromVector(Vector=$Particles.ToNozzle, Channel=::NewEnumerator1)
      Z=0
     )
     Safe Normalize=true
     Safe Fallback Threshold=0
     Safe Fallback Vector=<0 0 0.001>
    )) {isSetParameters=true}
   /SetVariables(Particles.Velocity=&MakeVector(
     X=&Multiply_Float(
      A=&MakeFloatFromVector(Vector=$Particles.ToNozzleHorizontal, Channel=::NewEnumerator0)
      B=&Lerp_Float(
       A=80
       B=0
       Alpha=&FloatFromCurve(
        FloatCurve=DataInterface() {propertyValues {Curve {defaultValue=0, keys [{tangentMode=RCTM_User, time=0, value=0}, {tangentMode=RCTM_User, time=1, value=1}]}}}
        Scale Curve=&Lerp_Float(A=1, B=0, Alpha=$User.ParaGlideLeanAlpha)
       )
      )
     )
     Y=&Multiply_Float(
      A=&MakeFloatFromVector(Vector=$Particles.ToNozzleHorizontal, Channel=::NewEnumerator1)
      B=&Lerp_Float(
       A=80
       B=0
       Alpha=&FloatFromCurve(
        FloatCurve=DataInterface() {propertyValues {Curve {
           defaultValue=0
           keys [
            {arriveTangent=9.0725, interpMode=RCIM_Cubic, leaveTangent=9.0725, tangentMode=RCTM_User, time=0, value=0}
            {arriveTangent=0.0003, interpMode=RCIM_Cubic, leaveTangent=0.0003, tangentMode=RCTM_User, time=0.2, value=1}
            {arriveTangent=0.0356, interpMode=RCIM_Cubic, leaveTangent=0.0356, tangentMode=RCTM_User, time=0.7, value=1}
            {arriveTangent=-7.4625, interpMode=RCIM_Cubic, leaveTangent=-7.4625, tangentMode=RCTM_User, time=1, value=0}
           ]
          }}}
        Scale Curve=&Lerp_Float(A=1, B=0, Alpha=$User.ParaGlideLeanAlpha)
       )
      )
     )
     Z=&Add_Float(
      A=&MakeFloatFromVector(Vector=$Output.UpdateParticlesFromOtherEmitter.SampledVelocity, Channel=::NewEnumerator2)
      B=&Lerp_Float(
       A=150
       B=&Lerp_Float(A=350, B=&FN_FrontEndFloatModifier(Gameplay=550, Front End=500), Alpha=$User.ParaGlideLeanAlpha)
       Alpha=&FloatFromCurve(
        FloatCurve=DataInterface() {propertyValues {Curve {defaultValue=1, keys [{tangentMode=RCTM_User, time=0, value=1}, <at 0.5 1>, {tangentMode=RCTM_User, time=0.9, value=0}]}}}
        CurveIndex=&Divide_Float(A=$Output.UpdateParticlesFromOtherEmitter.SampledAge, B=$Output.UpdateParticlesFromOtherEmitter.SampledLifetime)
       )
      )
     )
    )) {isSetParameters=true}
   /SetVariables(Particles.Velocity=&Add_Vector(
     A=$Particles.Velocity
     B=&FN_FrontEndVector3Modifier(
      Gameplay=<0 0 0>
      Front End=&MakeVector(
       X=&Multiply_Float(A=&MakeFloatFromVector(Vector=$System.Spraycan velocity, Channel=::NewEnumerator0), B=-450)
       Y=&Multiply_Float(A=&MakeFloatFromVector(Vector=$System.Spraycan velocity, Channel=::NewEnumerator1), B=-450)
       Z=0
      )
     )
    )) {isSetParameters=true}
   &ScaleColor(
    Scale Alpha=&FloatFromCurve(FloatCurve=DataInterface() {propertyValues {Curve {keys [
         <at 0 0 RCIM_Cubic>
         {arriveTangent=1.6041, leaveTangent=1.6041, time=0.35, value=1.0113}
         {arriveTangent=-2.2474, interpMode=RCIM_Cubic, leaveTangent=-2.2474, time=0.7, value=0.8}
         {interpMode=RCIM_Cubic, leaveTangent=-0.6038, time=0.8, value=0}
        ]}}})
    Scale RGB=&FN_FrontEndVector3Modifier(Gameplay=<1 1 1>, Front End=<1 1 1>)
    ScaleA=true
    ScaleRGB=true
   )
   &SolveForcesAndVelocity(Clamp Velocity=false, Limit Acceleration=false)
   &DynamicMaterialParameters(
    Index 0 Param 1=&RandomRangeFloat(Minimum=0, Maximum=1)
    Index 0 Param 2=&FloatFromCurve(
     FloatCurve=DataInterface() {propertyValues {Curve {keys [{interpMode=RCIM_Cubic, tangentMode=RCTM_User, time=0.33, value=0}, {interpMode=RCIM_Cubic, leaveTangent=1.2673, time=0.8, value=1}]}}}
     Scale Curve=0
    )
    Index 0 Param 3=&FloatFromCurve(FloatCurve=DataInterface() {propertyValues {Curve {keys [{tangentMode=RCTM_User, time=0, value=1}, {tangentMode=RCTM_User, time=1, value=1}]}}})
    Index 0 Param 4=&FloatFromCurve(FloatCurve=DataInterface() {propertyValues {Curve {keys [
         {arriveTangent=-14.9399, interpMode=RCIM_Cubic, leaveTangent=-14.9399, tangentMode=RCTM_User, time=0, value=1}
         {interpMode=RCIM_Cubic, tangentMode=RCTM_User, time=0.15, value=0}
         {interpMode=RCIM_Cubic, tangentMode=RCTM_User, time=0.33, value=0}
         {arriveTangent=2.7964, interpMode=RCIM_Cubic, leaveTangent=2.7964, tangentMode=RCTM_User, time=1, value=1}
        ]}}})
    Index 1 Param 2=&FloatFromCurve(FloatCurve=DataInterface() {propertyValues {Curve {keys [
         {arriveTangent=23.2225, interpMode=RCIM_Cubic, leaveTangent=23.2225, tangentMode=RCTM_User, time=0, value=0}
         <at 0.1 1 RCIM_Cubic>
         {arriveTangent=-3.2993, interpMode=RCIM_Cubic, leaveTangent=-3.2993, tangentMode=RCTM_User, time=1, value=0}
        ]}}})
    Module.Param0WriteEnabled=true
    Module.Param0WriteEnabled1=true
    Module.Param1WriteEnabled=true
    Module.Param1WriteEnabled1=true
    Module.Param2WriteEnabled=true
    Module.Param2WriteEnabled1=false
    Module.Param3WriteEnabled=true
    Module.Param3WriteEnabled1=false
    Write Parameter Index 1=true
   ) {module=DynamicMaterialParameters001}
  ]
  rendererClasses [&NiagaraSpriteRendererProperties]
  renderers [
   {rendererClass=&NiagaraSpriteRendererProperties}
   {rendererClass=&NiagaraSpriteRendererProperties, rendererIndex=1}
  ]
 }
 Liquid_trail001 {
  emitterUpdateScript [
   &EmitterState()
   &SpawnParticlesFromOtherEmitter(
    Attribute Reader=DataInterface() {propertyValues {EmitterBinding {bindingMode=Other, emitterName=Spraycan}}}
    Calculate Spawn Rate Per Particle=false
    Spawn Rate=30
    Use Maximum Age=false
    Use Minimum Age=false
   )
  ]
  particleSpawnScript [
   &InitializeParticle(
    Color Mode=::NewEnumerator0
    Lifetime=&Lerp_Float(A=0.5, B=1.2, Alpha=$User.ParaGlideLeanAlpha)
    Ribbon Width=35
    Ribbon Width Mode=::NewEnumerator1
    Sprite Rotation Mode=::NewEnumerator2
    Sprite Size Mode=::NewEnumerator3
    Sprite UV Mode=::NewEnumerator3
    Uniform Sprite Size=&Multiply_Float(
     A=&RandomRangeFloat(Minimum=40, Maximum=50)
     B=&self:SpawnGrp_0_1_Mask_05(Masked Float:NiagaraFloat=1, Passthrough Float:NiagaraFloat=1.2)
    )
    UsePositionOffset=false
    Write Material Random=true
   )
   &SampleParticlesFromOtherEmitter(
    Applied Velocity Scale=0
    Color Sampling=::NewEnumerator0
    Maximum Sampled Particle Age=1.6
    Mesh Orientation Sampling:ENiagara_DirectReadSamplingMode=::NewEnumerator0
    Minimum Sampled Particle Age=0.1
    Module.MaxSampledAgeCutoff=true
    Module.MinSampledAgeCutoff=false
    Particle ID Sampling=::NewEnumerator1
    Reject Particle By Sampled Age=true
    Ribbon UV Distance:ENiagara_DirectReadSamplingMode=::NewEnumerator0
    Velocity Sampling=::NewEnumerator0
   )
   /SetVariables(Particles.Position=&AddVectorToPosition(
     Position=$Particles.Position
     Vector=&Multiply_VectorByFloat(
      Vector=&ReturnMeshOrientationAxis(Mesh Orientation Axis=::ZAxis, Mesh Orientation Quat=$Particles.MeshOrientation)
      Float=13.9008
     )
    )) {isSetParameters=true, module=SetVariables_#5}
   /SetVariables(Particles.SprayDirection:Vector3f=&ReturnMeshOrientationAxis(Mesh Orientation Axis=::YAxis, Mesh Orientation Quat=$Particles.MeshOrientation)) {isSetParameters=true, module=SetVariables_#6}
   /SetVariables(Particles.SprayDirection:Vector3f=&NormalizeVector(
     Vector To Normalize=&MakeVector(
      X=&MakeFloatFromVector(Vector=$Particles.SprayDirection, Channel=::NewEnumerator0)
      Y=&MakeFloatFromVector(Vector=$Particles.SprayDirection, Channel=::NewEnumerator1)
      Z=0
     )
     Safe Normalize=true
     Safe Fallback Threshold=0
     Safe Fallback Vector=<0 0 0.001>
    )) {isSetParameters=true, module=SetVariables_#7}
  ]
  particleUpdateScript [
   &ParticleState(Lifetime:NiagaraFloat=&Lerp_Float(
     A=&FloatFromCurve(FloatCurve=DataInterface() {propertyValues {Curve {defaultValue=1, keys [{tangentMode=RCTM_User, time=0, value=0}, {tangentMode=RCTM_User, time=1, value=0.5}]}}})
     B=1
     Alpha=0.5
    ))
   &UpdateParticlesFromOtherEmitter(
    Age Sampling:ENiagara_DirectReadSamplingMode=::NewEnumerator1
    Attribute Reader=DataInterface() {propertyValues {EmitterBinding {bindingMode=Other, emitterName=Spraycan}}}
    Lifetime Sampling:ENiagara_DirectReadSamplingMode=::NewEnumerator1
    Mesh Orientation Sampling:ENiagara_DirectReadSamplingMode=::NewEnumerator0
    Mesh Scale Sampling:ENiagara_DirectReadSamplingMode=::NewEnumerator1
    Position Sampling:ENiagara_DirectReadSamplingMode=::NewEnumerator1
    Velocity Sampling:ENiagara_DirectReadSamplingMode=::NewEnumerator1
   )
   &InheritSourceMovement(Applied Position Scale=<-0.1 -0.1 0>, ApplyForce=false, ApplyPosition=true, ApplyPositionFalloff=false)
   &ScaleRibbonWidth(Ribbon Width Scale=&FloatFromCurve(
     FloatCurve=DataInterface() {propertyValues {Curve {
        defaultValue=0
        keys [{interpMode=RCIM_Cubic, tangentMode=RCTM_User, time=0, value=0}, {interpMode=RCIM_Cubic, tangentMode=RCTM_User, time=1, value=1}]
       }}}
     Scale Curve=&Lerp_Float(
      A=3.5
      B=0.5
      Alpha=&Divide_Float(A=$Output.UpdateParticlesFromOtherEmitter.SampledAge, B=$Output.UpdateParticlesFromOtherEmitter.SampledLifetime)
     )
    ))
   /SetVariables(Particles.ToNozzle:Vector3f=&ReturnMeshOrientationAxis(Mesh Orientation Axis=::ZAxis, Mesh Orientation Quat=$Particles.MeshOrientation)) {isSetParameters=true, module=SetVariables_#8}
   /SetVariables(Particles.ToNozzleHorizontal:Vector3f=&NormalizeVector(
     Vector To Normalize=&MakeVector(
      X=&MakeFloatFromVector(Vector=$Particles.ToNozzle, Channel=::NewEnumerator0)
      Y=&MakeFloatFromVector(Vector=$Particles.ToNozzle, Channel=::NewEnumerator1)
      Z=0
     )
     Safe Normalize=true
     Safe Fallback Threshold=0
     Safe Fallback Vector=<0 0 0.001>
    )) {isSetParameters=true, module=SetVariables_#9}
   /SetVariables(Particles.Velocity=&MakeVector(
     X=&Multiply_Float(
      A=&MakeFloatFromVector(Vector=$Particles.ToNozzleHorizontal, Channel=::NewEnumerator0)
      B=&Lerp_Float(
       A=80
       B=0
       Alpha=&FloatFromCurve(
        FloatCurve=DataInterface() {propertyValues {Curve {defaultValue=0, keys [{tangentMode=RCTM_User, time=0, value=0}, {tangentMode=RCTM_User, time=1, value=1}]}}}
        Scale Curve=&Lerp_Float(A=1, B=0, Alpha=$User.ParaGlideLeanAlpha)
       )
      )
     )
     Y=&Multiply_Float(
      A=&MakeFloatFromVector(Vector=$Particles.ToNozzleHorizontal, Channel=::NewEnumerator1)
      B=&Lerp_Float(
       A=80
       B=0
       Alpha=&FloatFromCurve(
        FloatCurve=DataInterface() {propertyValues {Curve {
           defaultValue=0
           keys [
            {arriveTangent=9.0725, interpMode=RCIM_Cubic, leaveTangent=9.0725, tangentMode=RCTM_User, time=0, value=0}
            {arriveTangent=0.0003, interpMode=RCIM_Cubic, leaveTangent=0.0003, tangentMode=RCTM_User, time=0.2, value=1}
            {arriveTangent=0.0356, interpMode=RCIM_Cubic, leaveTangent=0.0356, tangentMode=RCTM_User, time=0.7, value=1}
            {arriveTangent=-7.4625, interpMode=RCIM_Cubic, leaveTangent=-7.4625, tangentMode=RCTM_User, time=1, value=0}
           ]
          }}}
        Scale Curve=&Lerp_Float(A=1, B=0, Alpha=$User.ParaGlideLeanAlpha)
       )
      )
     )
     Z=&Add_Float(
      A=&MakeFloatFromVector(Vector=$Output.UpdateParticlesFromOtherEmitter.SampledVelocity, Channel=::NewEnumerator2)
      B=&Lerp_Float(
       A=150
       B=&Lerp_Float(A=350, B=&FN_FrontEndFloatModifier(Gameplay=550, Front End=500), Alpha=$User.ParaGlideLeanAlpha)
       Alpha=&FloatFromCurve(
        FloatCurve=DataInterface() {propertyValues {Curve {defaultValue=1, keys [{tangentMode=RCTM_User, time=0, value=1}, <at 0.5 1>, {tangentMode=RCTM_User, time=0.9, value=0}]}}}
        CurveIndex=&Divide_Float(A=$Output.UpdateParticlesFromOtherEmitter.SampledAge, B=$Output.UpdateParticlesFromOtherEmitter.SampledLifetime)
       )
      )
     )
    )) {isSetParameters=true}
   /SetVariables(Particles.Velocity=&Add_Vector(
     A=$Particles.Velocity
     B=&FN_FrontEndVector3Modifier(
      Gameplay=<0 0 0>
      Front End=&MakeVector(
       X=&Multiply_Float(A=&MakeFloatFromVector(Vector=$System.Spraycan velocity, Channel=::NewEnumerator0), B=-450)
       Y=&Multiply_Float(A=&MakeFloatFromVector(Vector=$System.Spraycan velocity, Channel=::NewEnumerator1), B=-450)
       Z=0
      )
     )
    )) {isSetParameters=true}
   &ScaleColor(
    Scale Alpha=&FloatFromCurve(FloatCurve=DataInterface() {propertyValues {Curve {
        defaultValue=0
        keys [
         {interpMode=RCIM_Cubic, tangentMode=RCTM_User, time=0, value=1}
         {arriveTangent=-0.9732, interpMode=RCIM_Cubic, leaveTangent=-0.9732, tangentMode=RCTM_User, time=1, value=0.5}
        ]
       }}})
    Scale RGB=&FN_FrontEndVector3Modifier(Gameplay=<1 1 1>, Front End=<1 1 1>)
    ScaleA=true
    ScaleRGB=true
   ) {module=ScaleColor001}
   &ScaleSpriteSize(Uniform Curve Sprite Scale=DataInterface() {propertyValues {Curve {keys [
        {arriveTangent=3.5003, interpMode=RCIM_Cubic, leaveTangent=3.5003, tangentMode=RCTM_User, time=0, value=0}
        {arriveTangent=-0.3385, interpMode=RCIM_Cubic, leaveTangent=-0.3385, tangentMode=RCTM_User, time=0.18, value=1}
        <at 1 1 RCIM_Cubic>
       ]}}})
   &SolveForcesAndVelocity(Clamp Velocity=false, Limit Acceleration=false)
   &DynamicMaterialParameters(
    Index 0 Param 1=&MaskFloatBySpawnGroup(Masked Float=2.2288, Passthrough Float=18.6053, Spawn Group Mask=1)
    Index 0 Param 2=&FloatFromCurve(FloatCurve=DataInterface() {propertyValues {Curve {keys [
         <at 0 0 RCIM_Cubic>
         {arriveTangent=-0.3062, interpMode=RCIM_Cubic, leaveTangent=-0.3062, tangentMode=RCTM_User, time=0.4, value=0.35}
         {arriveTangent=0.0167, interpMode=RCIM_Cubic, leaveTangent=0.0167, tangentMode=RCTM_Break, time=0.8, value=0}
        ]}}})
    Index 0 Param 3=1
    Index 0 Param 4=1
    Module.Index0Float0Write=true
    Module.Index0Float1Write=true
    Module.Index0Float2Write=true
    Module.Index0Float3Write=true
   ) {module=DynamicMaterialParameters002}
  ]
  rendererClasses [&NiagaraRibbonRendererProperties]
  renderers [{rendererClass=&NiagaraRibbonRendererProperties}]
 }
 Smoke001 {
  emitterUpdateScript [
   &EmitterState()
   &SpawnParticlesFromOtherEmitter(
    Attribute Reader=DataInterface() {propertyValues {EmitterBinding {bindingMode=Other, emitterName=Spraycan}}}
    Spawn Rate=&Lerp_Float(A=60, B=65, Alpha=$User.ParaGlideLeanAlpha)
    Use Maximum Age=false
    Use Minimum Age=false
   )
  ]
  particleSpawnScript [
   &InitializeParticle(
    Color=<rgba 1 1 1 0.2>
    Lifetime=&Lerp_Float(A=0.5, B=1.2, Alpha=$User.ParaGlideLeanAlpha)
    Ribbon Width=45
    Ribbon Width Mode=::NewEnumerator1
    Sprite Rotation Mode=::NewEnumerator2
    Sprite Size Mode=::NewEnumerator3
    Sprite UV Mode=::NewEnumerator3
    Uniform Sprite Size=&Multiply_Float(
     A=&RandomRangeFloat(Minimum=25, Maximum=40)
     B=&self:SpawnGrp_0_1_Mask_04(Masked Float:NiagaraFloat=1, Passthrough Float:NiagaraFloat=1.2)
    )
    UsePositionOffset=false
    Write Material Random=true
   )
   &SampleParticlesFromOtherEmitter(
    Age Sampling:ENiagara_DirectReadSamplingMode=::NewEnumerator1
    Applied Velocity Scale=0
    Color Sampling=::NewEnumerator0
    Lifetime Sampling:ENiagara_DirectReadSamplingMode=::NewEnumerator1
    Maximum Sampled Particle Age=1.7
    Mesh Orientation Sampling:ENiagara_DirectReadSamplingMode=::NewEnumerator0
    Minimum Sampled Particle Age=0.1
    Module.MaxSampledAgeCutoff=true
    Module.MinSampledAgeCutoff=false
    Particle ID Sampling=::NewEnumerator1
    Reject Particle By Sampled Age=true
    Ribbon UV Distance:ENiagara_DirectReadSamplingMode=::NewEnumerator0
    Velocity Sampling=::NewEnumerator0
   )
   /SetVariables(Particles.Position=&AddVectorToPosition(
     Position=$Particles.Position
     Vector=&Multiply_VectorByFloat(
      Vector=&ReturnMeshOrientationAxis(Mesh Orientation Axis=::ZAxis, Mesh Orientation Quat=$Particles.MeshOrientation)
      Float=13.9
     )
    )) {isSetParameters=true, module=SetVariables_#5}
   /SetVariables(Particles.SprayDirection:Vector3f=&ReturnMeshOrientationAxis(Mesh Orientation Axis=::YAxis, Mesh Orientation Quat=$Particles.MeshOrientation)) {isSetParameters=true, module=SetVariables_#6}
   /SetVariables(Particles.SprayDirection:Vector3f=&NormalizeVector(
     Vector To Normalize=&MakeVector(
      X=&MakeFloatFromVector(Vector=$Particles.SprayDirection, Channel=::NewEnumerator0)
      Y=&MakeFloatFromVector(Vector=$Particles.SprayDirection, Channel=::NewEnumerator1)
      Z=0
     )
     Safe Normalize=true
     Safe Fallback Threshold=0
     Safe Fallback Vector=<0 0 0.001>
    )) {isSetParameters=true, module=SetVariables_#7}
  ]
  particleUpdateScript [
   &ParticleState(Lifetime:NiagaraFloat=&Lerp_Float(
     A=&FloatFromCurve(FloatCurve=DataInterface() {propertyValues {Curve {defaultValue=1, keys [{tangentMode=RCTM_User, time=0, value=0}, {tangentMode=RCTM_User, time=1, value=0.5}]}}})
     B=1
     Alpha=0.5
    ))
   &InheritSourceMovement(Applied Position Scale=<-0.1 -0.1 0>, ApplyForce=false, ApplyPosition=true, ApplyPositionFalloff=false)
   &UpdateParticlesFromOtherEmitter(
    Age Sampling:ENiagara_DirectReadSamplingMode=::NewEnumerator1
    Attribute Reader=DataInterface() {propertyValues {EmitterBinding {bindingMode=Other, emitterName=Spraycan}}}
    Lifetime Sampling:ENiagara_DirectReadSamplingMode=::NewEnumerator1
    Mesh Orientation Sampling:ENiagara_DirectReadSamplingMode=::NewEnumerator0
    Position Sampling:ENiagara_DirectReadSamplingMode=::NewEnumerator1
    Velocity Sampling:ENiagara_DirectReadSamplingMode=::NewEnumerator1
   )
   &ScaleSpriteSize(Uniform Curve Sprite Scale=DataInterface() {propertyValues {Curve {keys [
        {interpMode=RCIM_Cubic, leaveTangent=1.046, tangentMode=RCTM_Break, time=0, value=0.3}
        {arriveTangent=3.3027, interpMode=RCIM_Cubic, leaveTangent=3.3027, tangentMode=RCTM_User, time=0.3, value=0.65}
        {tangentMode=RCTM_User, time=1, value=1.2}
       ]}}}) {module=ScaleSpriteSize001}
   /SetVariables(Particles.ToNozzle:Vector3f=&ReturnMeshOrientationAxis(Mesh Orientation Axis=::ZAxis, Mesh Orientation Quat=$Particles.MeshOrientation)) {isSetParameters=true, module=SetVariables_#8}
   /SetVariables(Particles.ToNozzleHorizontal:Vector3f=&NormalizeVector(
     Vector To Normalize=&MakeVector(
      X=&MakeFloatFromVector(Vector=$Particles.ToNozzle, Channel=::NewEnumerator0)
      Y=&MakeFloatFromVector(Vector=$Particles.ToNozzle, Channel=::NewEnumerator1)
      Z=0
     )
     Safe Normalize=true
     Safe Fallback Threshold=0
     Safe Fallback Vector=<0 0 0.001>
    )) {isSetParameters=true, module=SetVariables_#9}
   /SetVariables(Particles.Velocity=&MakeVector(
     X=&Multiply_Float(
      A=&MakeFloatFromVector(Vector=$Particles.ToNozzleHorizontal, Channel=::NewEnumerator0)
      B=&Lerp_Float(
       A=80
       B=0
       Alpha=&FloatFromCurve(
        FloatCurve=DataInterface() {propertyValues {Curve {defaultValue=0, keys [{tangentMode=RCTM_User, time=0, value=0}, {tangentMode=RCTM_User, time=1, value=1}]}}}
        Scale Curve=&Lerp_Float(A=1, B=0, Alpha=$User.ParaGlideLeanAlpha)
       )
      )
     )
     Y=&Multiply_Float(
      A=&MakeFloatFromVector(Vector=$Particles.ToNozzleHorizontal, Channel=::NewEnumerator1)
      B=&Lerp_Float(
       A=80
       B=0
       Alpha=&FloatFromCurve(
        FloatCurve=DataInterface() {propertyValues {Curve {
           defaultValue=0
           keys [
            {arriveTangent=9.0725, interpMode=RCIM_Cubic, leaveTangent=9.0725, tangentMode=RCTM_User, time=0, value=0}
            {arriveTangent=0.0003, interpMode=RCIM_Cubic, leaveTangent=0.0003, tangentMode=RCTM_User, time=0.2, value=1}
            {arriveTangent=0.0356, interpMode=RCIM_Cubic, leaveTangent=0.0356, tangentMode=RCTM_User, time=0.7, value=1}
            {arriveTangent=-7.4625, interpMode=RCIM_Cubic, leaveTangent=-7.4625, tangentMode=RCTM_User, time=1, value=0}
           ]
          }}}
        Scale Curve=&Lerp_Float(A=1, B=0, Alpha=$User.ParaGlideLeanAlpha)
       )
      )
     )
     Z=&Add_Float(
      A=&MakeFloatFromVector(Vector=$Output.UpdateParticlesFromOtherEmitter.SampledVelocity, Channel=::NewEnumerator2)
      B=&Lerp_Float(
       A=150
       B=&Lerp_Float(A=350, B=&FN_FrontEndFloatModifier(Gameplay=550, Front End=500), Alpha=$User.ParaGlideLeanAlpha)
       Alpha=&FloatFromCurve(
        FloatCurve=DataInterface() {propertyValues {Curve {defaultValue=1, keys [{tangentMode=RCTM_User, time=0, value=1}, <at 0.5 1>, {tangentMode=RCTM_User, time=0.9, value=0}]}}}
        CurveIndex=&Divide_Float(A=$Output.UpdateParticlesFromOtherEmitter.SampledAge, B=$Output.UpdateParticlesFromOtherEmitter.SampledLifetime)
       )
      )
     )
    )) {isSetParameters=true}
   /SetVariables(Particles.Velocity=&Add_Vector(
     A=$Particles.Velocity
     B=&FN_FrontEndVector3Modifier(
      Gameplay=<0 0 0>
      Front End=&MakeVector(
       X=&Multiply_Float(A=&MakeFloatFromVector(Vector=$System.Spraycan velocity, Channel=::NewEnumerator0), B=-450)
       Y=&Multiply_Float(A=&MakeFloatFromVector(Vector=$System.Spraycan velocity, Channel=::NewEnumerator1), B=-450)
       Z=0
      )
     )
    )) {isSetParameters=true}
   &ScaleColor(
    Scale Alpha=&FloatFromCurve(
     FloatCurve=DataInterface() {propertyValues {Curve {keys [
         {interpMode=RCIM_Cubic, tangentMode=RCTM_User, time=0, value=0}
         {arriveTangent=-0.4195, interpMode=RCIM_Cubic, leaveTangent=-0.4195, tangentMode=RCTM_User, time=0.3, value=0.7}
         {arriveTangent=-0.0541, interpMode=RCIM_Cubic, leaveTangent=-0.0541, tangentMode=RCTM_User, time=0.85, value=0}
        ]}}}
     Scale Curve=&Lerp_Float(
      A=&FNTimeOfDayScalar(Night Scale=0.85, Morning Scale=0.45, Day Scale=0.5, Evening Scale=0.45)
      B=0
      Alpha=&Divide_Float(A=$Output.UpdateParticlesFromOtherEmitter.SampledAge, B=$Output.UpdateParticlesFromOtherEmitter.SampledLifetime)
     )
    )
    Scale RGB=&FN_FrontEndVector3Modifier(Gameplay=<2 2 2>, Front End=<2 2 2>)
    ScaleA=true
    ScaleRGB=true
   )
   &SolveForcesAndVelocity(Clamp Velocity=false, Limit Acceleration=false)
   &DynamicMaterialParameters(
    Index 0 Param 1=0.5
    Index 0 Param 2=&FloatFromCurve(FloatCurve=DataInterface() {propertyValues {Curve {keys [
         {interpMode=RCIM_Cubic, tangentMode=RCTM_Break, time=0, value=1}
         {arriveTangent=-2.6912, interpMode=RCIM_Cubic, leaveTangent=-2.0957, tangentMode=RCTM_Break, time=0.7, value=0.3}
         <at 1 0>
        ]}}})
    Index 0 Param 3=&FloatFromCurve(FloatCurve=DataInterface() {propertyValues {Curve {keys [
         {arriveTangent=-2.3777, interpMode=RCIM_Cubic, tangentMode=RCTM_Break, time=0, value=1.35}
         {arriveTangent=0.0149, interpMode=RCIM_Cubic, tangentMode=RCTM_Break, time=1, value=0.4}
        ]}}})
    Module.Index0Float0Write=true
    Module.Index0Float1Write=true
    Module.Index0Float2Write=true
    Module.Index0Float3Write=true
   )
   /Niagara/Modules/Update/SubUV/V2/SubUVAnimation(
    Playback Scale Curve=DataInterface() {propertyValues {Curve {keys [
        {interpMode=RCIM_Cubic, tangentMode=RCTM_User, time=0.25, value=1}
        {arriveTangent=-1.2656, interpMode=RCIM_Cubic, leaveTangent=-1.2656, tangentMode=RCTM_User, time=1, value=0}
       ]}}}
    Random Start Frame=true
    "Sprite Renderer "=DataInterface() {propertyValues {SpriteRenderer=&self:Smoke001_0.NiagaraSpriteRendererProperties_2}}
    SubUV Animation Mode:ENiagara_SubUVLookupModeV2=::NewEnumerator3
    UseEndFrame=false
    UsePlayRateScale=true
    UseStartFrame=false
   )
  ]
  rendererClasses [&NiagaraSpriteRendererProperties]
  renderers [{rendererClass=&NiagaraSpriteRendererProperties}]
 }
 Spraycan {
  emitterUpdateScript [
   &EmitterState()
   &SpawnRate(
    SpawnGroup:NiagaraInt32=&RandomRangeInt(Minimum=0, Maximum=2)
    SpawnRate=&Lerp_Float(A=1.5, B=3, Alpha=$User.ParaGlideLeanAlpha)
    Use Spawn Probability=false
   )
  ]
  particleSpawnScript [
   /SetVariables(Particles.VisibilityTag=&SpawnGroup(Max Spawn Group=3)) {isSetParameters=true}
   &InitializeParticle(
    AdjustAlpha=false
    AdjustHue=false
    AdjustSaturation=false
    AdjustValue=false
    Color=&SelectLinearColorFromArray(
     Array Sampling Mode=::NewEnumerator1
     Direct Array Index=$Particles.VisibilityTag
     Direct Array Mode=::NewEnumerator1
     Color Selection Array=DataInterface() {propertyValues {ColorData [<rgba 1 0.0782 0.8714 1>, <rgba 1 0.6445 0 1>, <rgba 0.0704 0.3916 0.2307 1>]}}
     Random Seed=?
    )
    Color Mode=::NewEnumerator3
    Hue Shift Range:Vector2f=<0 0.1>
    Lifetime=1.7
    Mesh Scale Mode=::NewEnumerator3
    Position Mode=::NewEnumerator1
    Position Offset=<0 0 -50>
    UsePositionOffset=true
    Value Range:Vector2f=<0 1>
    Write Material Random=true
   )
   &InitialMeshOrientation(
    Mesh Axis To Orient=<0 0 1>
    Mesh Orientation Mode=::NewEnumerator2
    Orientation Vector=<0 0 1>
    Rotation=&RandomRangeVector(Minimum=<0 0 0>, Maximum=<0 0 1>)
    Use Rotation Vector=true
   ) {module=InitialMeshOrientation001}
   /SetVariables(Particles.InitialOrientation:NiagaraFloat=&RandomRangeFloat(Minimum=-1, Maximum=1)) {isSetParameters=true}
   /SetVariables(Particles.CupID:NiagaraFloat=&MakeFloatFromInt(Int=&ReturnParticleIDAttributes())) {isSetParameters=true}
  ]
  particleUpdateScript [
   &ParticleState()
   &ShapeLocation(
    MaskSpawnGroup=false
    Offset=&FN_FrontEndVector3Modifier(Gameplay=<0 25 -50>, Front End=&Lerp_Vector(A=<0 25 -50>, B=<0 75 -50>, Alpha=$User.ParaGlideLeanAlpha))
    Offset Coordinate Space=::Local
    Offset Mode=::NewEnumerator0
    Override Seed=false
    Radius Position=1.75
    Ring / Disc Distribution Mode:ENiagaraSphereDistributionMode=::NewEnumerator1
    Ring / Disc Mode:ENiagaraRingDiscMode=::NewEnumerator1
    Ring Radius=&Lerp_Float(
     A=&FN_FrontEndFloatModifier(Gameplay=80, Front End=68)
     B=&FN_FrontEndFloatModifier(Gameplay=50, Front End=60)
     Alpha=$User.ParaGlideLeanAlpha
    )
    Rotation Angle:NiagaraFloat=&RandomRangeFloat(Minimum=80, Maximum=100)
    Rotation Angle Type:ENiagara_AngleInput=::NewEnumerator0
    Rotation Axis=<0 0 1>
    Rotation Mode=::NewEnumerator1
    Shape Primitive=::NewEnumerator4
    U Position=&Add_Float(
     A=&Multiply_Float(A=$Particles.Initial.CupID, B=0.6)
     B=&FloatFromCurve(
      FloatCurve=DataInterface() {propertyValues {Curve {keys [{tangentMode=RCTM_User, time=-1, value=0}, {tangentMode=RCTM_User, time=1, value=1}]}}}
      Scale Curve=0.05
     )
    )
    UseOwnerScale=true
   )
   /SetVariables(Particles.MeshOrientation=&MultiplyQuaternion(
     Quaternion A=&FindQuatBetweenTwoVectors(
      From Vector=&NormalizeVector(
       Vector To Normalize=&MakeVector(
        X=&Sine(
         Normalized Angle=&Add_Float(A=$Emitter.Age, B=$Particles.Initial.InitialOrientation)
         Period=&MakeCustomFloatFromBool(Bool=&RandomBool(), True Float=1, False Float=-1)
         Scale=0.5
         Bias=0
        )
        Y=&Cosine(
         Normalized Angle=&Add_Float(A=$Emitter.Age, B=$Particles.Initial.InitialOrientation)
         Period=&MakeCustomFloatFromBool(Bool=&RandomBool(), True Float=1, False Float=-1)
         Scale=0.5
         Bias=0
        )
        Z=&FloatFromCurve(FloatCurve=DataInterface() {propertyValues {Curve {keys [{interpMode=RCIM_Cubic, tangentMode=RCTM_User, time=0.3, value=1}, {interpMode=RCIM_Cubic, tangentMode=RCTM_User, time=1, value=-1}]}}})
       )
       Safe Normalize=true
       Safe Fallback Threshold=0
       Safe Fallback Vector=<0 0 0.001>
      )
      To Vector=<0 0 1>
     )
     Quaternion B=$Particles.Initial.MeshOrientation
    )) {isSetParameters=true}
   /SetVariables(Particles.Position=&Add_Vector(
     A=$Particles.ShapeLocation.ShapePosition
     B=&Multiply_VectorByFloat(
      Vector=&Add_Vector(A=$System.Spraycan velocity, B=<0 0 -0.25>)
      Float=&FloatFromCurve(
       FloatCurve=DataInterface() {propertyValues {Curve {defaultValue=0, keys [{tangentMode=RCTM_User, time=0, value=0}, {tangentMode=RCTM_User, time=1, value=1}]}}}
       Scale Curve=$System.CupDistance
      )
     )
    )) {isSetParameters=true}
   &PointAttractionForce(Attraction Strength=5, Kill Within Radius=false, Module.Use Falloff=true, UseAttractorOffset=false)
   &InheritSourceMovement(Applied Position Scale=<-0.15 -0.15 0>, ApplyForce=false, ApplyPosition=true, ApplyPositionFalloff=false)
   &SolveForcesAndVelocity(Clamp Velocity=false, Limit Acceleration=false)
   &ScaleMeshSize(Scale Factor=&VectorFromFloat(Value=&FloatFromCurve(FloatCurve=DataInterface() {propertyValues {Curve {keys [
          {arriveTangent=2.0977, interpMode=RCIM_Cubic, leaveTangent=2.0977, tangentMode=RCTM_User, time=0, value=0}
          {interpMode=RCIM_Cubic, tangentMode=RCTM_User, time=0.15, value=1}
          {arriveTangent=0.0772, interpMode=RCIM_Cubic, leaveTangent=0.0772, tangentMode=RCTM_User, time=0.9, value=1}
          {arriveTangent=-5.7608, interpMode=RCIM_Cubic, leaveTangent=-5.7608, tangentMode=RCTM_User, time=1, value=0}
         ]}}})))
   &ScaleColor(
    Scale Alpha=&Lerp_Float(
     A=4
     B=0.4
     Alpha=&FloatFromCurve(FloatCurve=DataInterface() {propertyValues {Curve {keys [{interpMode=RCIM_Cubic, tangentMode=RCTM_User, time=0, value=0}, {interpMode=RCIM_Cubic, tangentMode=RCTM_User, time=0.8, value=1}]}}})
    )
    ScaleA=true
    ScaleRGB=true
   )
   /Game/Effects/Niagara/Modules/Materials/FN_MeshMaterial(
    Amplitude=2
    Falloff=-15
    Frequency=20
    Jitter Multiplier=3
    Jitter Pan by SpawnGroups=false
    Offset=10
    Scale Noise X Axis=2
    Scale Noise Y Axis=2
    Speed=1
   )
   /SetVariables(Particles.GreenSprayRandom:NiagaraInt32=&RandomRangeInt(Minimum=0, Maximum=1)) {isSetParameters=true}
  ]
  rendererClasses [&NiagaraMeshRendererProperties]
  renderers [
   {rendererClass=&NiagaraMeshRendererProperties, rendererIndex=0}
   {rendererClass=&NiagaraMeshRendererProperties, rendererIndex=1}
   {rendererClass=&NiagaraMeshRendererProperties, rendererIndex=2}
  ]
 }
 drops {
  emitterUpdateScript [
   &EmitterState()
   &SpawnParticlesFromOtherEmitter(
    Attribute Reader=DataInterface() {propertyValues {EmitterBinding {bindingMode=Other, emitterName=Spraycan}}}
    Calculate Spawn Rate Per Particle=false
    Spawn Rate=65
    Use Maximum Age=false
    Use Minimum Age=false
   )
  ]
  particleSpawnScript [
   &InitializeParticle(
    Color Mode=::NewEnumerator0
    Lifetime=&Lerp_Float(A=0.2, B=0.4, Alpha=$User.ParaGlideLeanAlpha)
    Sprite Rotation Mode=::NewEnumerator2
    Sprite Size Max=<25 40>
    Sprite Size Min=<5 20>
    Sprite Size Mode=::NewEnumerator2
    Sprite UV Mode=::NewEnumerator3
    UsePositionOffset=false
    Write Material Random=true
   )
   &SampleParticlesFromOtherEmitter(
    Applied Position Offset:Vector3f=<0.4352 0 0>
    Applied Velocity Scale=0
    Color Sampling=::NewEnumerator0
    Maximum Sampled Particle Age=1.6
    Mesh Orientation Sampling:ENiagara_DirectReadSamplingMode=::NewEnumerator0
    Minimum Sampled Particle Age=0.1
    Module.MaxSampledAgeCutoff=true
    Module.MinSampledAgeCutoff=false
    Particle ID Sampling=::NewEnumerator1
    Reject Particle By Sampled Age=true
    Ribbon UV Distance:ENiagara_DirectReadSamplingMode=::NewEnumerator0
    Velocity Sampling=::NewEnumerator0
   )
   &SubUVAnimation(
    End Frame Range Override=3
    "Sprite Renderer "=DataInterface() {propertyValues {SpriteRenderer=&self:drops_0.NiagaraSpriteRendererProperties_1}}
    SubUV Animation Mode:ENiagara_SubUVLookupModeV2=::NewEnumerator2
    UseEndFrame=true
    UseStartFrame=false
   ) {module=SubUVAnimation001}
   /SetVariables(Particles.Position=&AddVectorToPosition(
     Position=$Particles.Position
     Vector=&Multiply_VectorByFloat(
      Vector=&ReturnMeshOrientationAxis(Mesh Orientation Axis=::ZAxis, Mesh Orientation Quat=$Particles.MeshOrientation)
      Float=13.9008
     )
    )) {isSetParameters=true, module=SetVariables_#5}
   /SetVariables(Particles.SprayDirection:Vector3f=&ReturnMeshOrientationAxis(Mesh Orientation Axis=::YAxis, Mesh Orientation Quat=$Particles.MeshOrientation)) {isSetParameters=true, module=SetVariables_#6}
   /SetVariables(Particles.SprayDirection:Vector3f=&NormalizeVector(
     Vector To Normalize=&MakeVector(
      X=&MakeFloatFromVector(Vector=$Particles.SprayDirection, Channel=::NewEnumerator0)
      Y=&MakeFloatFromVector(Vector=$Particles.SprayDirection, Channel=::NewEnumerator1)
      Z=0
     )
     Safe Normalize=true
     Safe Fallback Threshold=0
     Safe Fallback Vector=<0 0 0.001>
    )) {isSetParameters=true, module=SetVariables_#7}
  ]
  particleUpdateScript [
   &ParticleState(Lifetime:NiagaraFloat=&Lerp_Float(
     A=&FloatFromCurve(FloatCurve=DataInterface() {propertyValues {Curve {defaultValue=1, keys [{tangentMode=RCTM_User, time=0, value=0}, {tangentMode=RCTM_User, time=1, value=0.5}]}}})
     B=1
     Alpha=0.5
    ))
   &UpdateParticlesFromOtherEmitter(
    Age Sampling:ENiagara_DirectReadSamplingMode=::NewEnumerator1
    Attribute Reader=DataInterface() {propertyValues {EmitterBinding {bindingMode=Other, emitterName=Spraycan}}}
    Lifetime Sampling:ENiagara_DirectReadSamplingMode=::NewEnumerator1
    Mesh Orientation Sampling:ENiagara_DirectReadSamplingMode=::NewEnumerator0
    Position Sampling:ENiagara_DirectReadSamplingMode=::NewEnumerator1
    Velocity Sampling:ENiagara_DirectReadSamplingMode=::NewEnumerator1
   )
   &InheritSourceMovement(Applied Position Scale=<-0.1 -0.1 0>, ApplyForce=false, ApplyPosition=true, ApplyPositionFalloff=false)
   &SpriteRotationRate()
   &ScaleSpriteSize(
    Uniform Curve Scale=&Lerp_Float(
     A=1.8164
     B=0
     Alpha=&Divide_Float(A=$Output.UpdateParticlesFromOtherEmitter.SampledAge, B=$Output.UpdateParticlesFromOtherEmitter.SampledLifetime)
    )
    Uniform Curve Sprite Scale=DataInterface() {propertyValues {Curve {keys [
        {interpMode=RCIM_Cubic, leaveTangent=0.2046, tangentMode=RCTM_Break, time=0, value=0}
        {arriveTangent=2.6398, interpMode=RCIM_Cubic, leaveTangent=2.6398, tangentMode=RCTM_User, time=0.4, value=0.6}
        {tangentMode=RCTM_User, time=1, value=1}
       ]}}}
   ) {module=ScaleSpriteSize001}
   /SetVariables(Particles.ToNozzle:Vector3f=&ReturnMeshOrientationAxis(Mesh Orientation Axis=::ZAxis, Mesh Orientation Quat=$Particles.MeshOrientation)) {isSetParameters=true, module=SetVariables_#8}
   /SetVariables(Particles.ToNozzleHorizontal:Vector3f=&NormalizeVector(
     Vector To Normalize=&MakeVector(
      X=&MakeFloatFromVector(Vector=$Particles.ToNozzle, Channel=::NewEnumerator0)
      Y=&MakeFloatFromVector(Vector=$Particles.ToNozzle, Channel=::NewEnumerator1)
      Z=0
     )
     Safe Normalize=true
     Safe Fallback Threshold=0
     Safe Fallback Vector=<0 0 0.001>
    )) {isSetParameters=true, module=SetVariables_#9}
   /SetVariables(Particles.Velocity=&MakeVector(
     X=&Multiply_Float(
      A=&MakeFloatFromVector(Vector=$Particles.ToNozzleHorizontal, Channel=::NewEnumerator0)
      B=&Lerp_Float(
       A=80
       B=0
       Alpha=&FloatFromCurve(
        FloatCurve=DataInterface() {propertyValues {Curve {defaultValue=0, keys [{tangentMode=RCTM_User, time=0, value=0}, {tangentMode=RCTM_User, time=1, value=1}]}}}
        Scale Curve=&Lerp_Float(A=1, B=0, Alpha=$User.ParaGlideLeanAlpha)
       )
      )
     )
     Y=&Multiply_Float(
      A=&MakeFloatFromVector(Vector=$Particles.ToNozzleHorizontal, Channel=::NewEnumerator1)
      B=&Lerp_Float(
       A=80
       B=0
       Alpha=&FloatFromCurve(
        FloatCurve=DataInterface() {propertyValues {Curve {
           defaultValue=0
           keys [
            {arriveTangent=9.0725, interpMode=RCIM_Cubic, leaveTangent=9.0725, tangentMode=RCTM_User, time=0, value=0}
            {arriveTangent=0.0003, interpMode=RCIM_Cubic, leaveTangent=0.0003, tangentMode=RCTM_User, time=0.2, value=1}
            {arriveTangent=0.0356, interpMode=RCIM_Cubic, leaveTangent=0.0356, tangentMode=RCTM_User, time=0.7, value=1}
            {arriveTangent=-7.4625, interpMode=RCIM_Cubic, leaveTangent=-7.4625, tangentMode=RCTM_User, time=1, value=0}
           ]
          }}}
        Scale Curve=&Lerp_Float(A=1, B=0, Alpha=$User.ParaGlideLeanAlpha)
       )
      )
     )
     Z=&Add_Float(
      A=&MakeFloatFromVector(Vector=$Output.UpdateParticlesFromOtherEmitter.SampledVelocity, Channel=::NewEnumerator2)
      B=&Lerp_Float(
       A=150
       B=&Lerp_Float(A=350, B=&FN_FrontEndFloatModifier(Gameplay=550, Front End=500), Alpha=$User.ParaGlideLeanAlpha)
       Alpha=&FloatFromCurve(
        FloatCurve=DataInterface() {propertyValues {Curve {defaultValue=1, keys [{tangentMode=RCTM_User, time=0, value=1}, <at 0.5 1>, {tangentMode=RCTM_User, time=0.9, value=0}]}}}
        CurveIndex=&Divide_Float(A=$Output.UpdateParticlesFromOtherEmitter.SampledAge, B=$Output.UpdateParticlesFromOtherEmitter.SampledLifetime)
       )
      )
     )
    )) {isSetParameters=true}
   /SetVariables(Particles.Velocity=&Add_Vector(
     A=$Particles.Velocity
     B=&FN_FrontEndVector3Modifier(
      Gameplay=<0 0 0>
      Front End=&MakeVector(
       X=&Multiply_Float(A=&MakeFloatFromVector(Vector=$System.Spraycan velocity, Channel=::NewEnumerator0), B=-400)
       Y=&Multiply_Float(A=&MakeFloatFromVector(Vector=$System.Spraycan velocity, Channel=::NewEnumerator1), B=-400)
       Z=0
      )
     )
    )) {isSetParameters=true}
   &ScaleColor(
    Scale Alpha=&FloatFromCurve(FloatCurve=DataInterface() {propertyValues {Curve {keys [
         <at 0 0 RCIM_Cubic>
         {arriveTangent=1.6041, leaveTangent=1.6041, time=0.05, value=1.0113}
         {arriveTangent=-1.3484, interpMode=RCIM_Cubic, leaveTangent=-1.3484, time=0.75, value=1}
         <at 0.8 0 RCIM_Cubic>
        ]}}})
    ScaleA=true
    ScaleRGB=true
   )
   &ScaleColor(
    Scale Alpha=&FloatFromCurve(FloatCurve=DataInterface() {propertyValues {Curve {keys [
         {interpMode=RCIM_Cubic, tangentMode=RCTM_User, time=0, value=0}
         {interpMode=RCIM_Cubic, tangentMode=RCTM_User, time=0.3, value=1}
         <at 0.75 0 RCIM_Cubic>
        ]}}})
    Scale RGB=&FN_FrontEndVector3Modifier(Gameplay=<1 1 1>, Front End=<1 1 1>)
    ScaleA=true
    ScaleRGB=true
   ) {module=ScaleColor001}
   &SolveForcesAndVelocity(Clamp Velocity=false, Limit Acceleration=false)
   &DynamicMaterialParameters(
    Index 0 Param 1=0.5
    Index 0 Param 2=&FloatFromCurve(FloatCurve=DataInterface() {propertyValues {Curve {keys [
         {interpMode=RCIM_Cubic, tangentMode=RCTM_Break, time=0, value=1}
         {arriveTangent=-2.6912, interpMode=RCIM_Cubic, leaveTangent=-2.0957, tangentMode=RCTM_Break, time=0.7, value=0.3}
         <at 1 0>
        ]}}})
    Index 0 Param 3=&FloatFromCurve(FloatCurve=DataInterface() {propertyValues {Curve {keys [
         {arriveTangent=-2.3777, interpMode=RCIM_Cubic, tangentMode=RCTM_Break, time=0, value=1.35}
         {arriveTangent=0.0149, interpMode=RCIM_Cubic, tangentMode=RCTM_Break, time=1, value=0.4}
        ]}}})
    Module.Index0Float0Write=true
    Module.Index0Float1Write=true
    Module.Index0Float2Write=true
    Module.Index0Float3Write=true
   )
  ]
  rendererClasses [&NiagaraSpriteRendererProperties]
  renderers [{rendererClass=&NiagaraSpriteRendererProperties}]
 }
}
summary {
 systemName=NS_Contrail_GraffitiTon_Trail_v3
 emitters [
  {emitterName=Spraycan, rendererClasses [&NiagaraMeshRendererProperties]}
  {emitterName=Smoke001, rendererClasses [&NiagaraSpriteRendererProperties]}
  {emitterName=Liquid_trail001, rendererClasses [&NiagaraRibbonRendererProperties]}
  {emitterName=drops, rendererClasses [&NiagaraSpriteRendererProperties]}
  {emitterName=Liquid_Particles001, rendererClasses [&NiagaraSpriteRendererProperties]}
 ]
 userVariables [
  User.bIsFrontEnd {type=NiagaraBool}
  User.bIsFrontEndPreview {type=NiagaraBool}
  User.bisSwinging {type=NiagaraBool}
  User.ColorVariant {type=NiagaraFloat, defaultValue=0, description=""}
  User.FrontEndScale {type=NiagaraFloat, defaultValue=0.8, description=""}
  User.ParaGlideLeanAlpha {type=NiagaraFloat, defaultValue=0, description=""}
  User.RibbonIDOffset {type=NiagaraInt32}
  User.Squeeze {type=NiagaraFloat, defaultValue=0, description=""}
  User.Trail {type=NiagaraFloat, defaultValue=0, description=""}
 ]
}
